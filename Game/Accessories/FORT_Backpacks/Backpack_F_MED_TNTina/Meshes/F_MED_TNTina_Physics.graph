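# Physics asset for the TNTina backpack, letting a string of grenade bones dangle and swing from the pack.
# The pack bone has a large box body, and the three dyn_grenade bones use two small boxes and a capsule at the tip.
# Constraints chain each grenade to the one before it, starting at the pack, with limited swing and tight twist to restrain the motion.
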PhysicsAsset /Game/Accessories/FORT_Backpacks/Backpack_F_MED_TNTina/Meshes/F_MED_TNTina_Physics
bodies {
 dyn_grenade_01 [{shapeType=Box, extentX=4, extentY=1.2541, extentZ=5.4711, center=<2.7 0 0>, rotation=<rot 90 0 0>}]
 dyn_grenade_02 [{shapeType=Box, extentX=4, extentY=1.2541, extentZ=5.4711, center=<2.7 0 0>, rotation=<rot 90 0 0>}]
 dyn_grenade_03 [{shapeType=Capsule, length=3.621, radius=2.6662, center=<4.3655 0 0>, rotation=<rot 90 0 0>}]
 pack [{shapeType=Box, extentX=21.6507, extentY=6.613, extentZ=19.619, center=<0.2141 -3.274 1.6684>, rotation=<rot 70.1968 0 0>}]
}
constraints [
 {bone1Name=dyn_grenade_01, bone2Name=pack, swing1LimitDegrees=35, swing2LimitDegrees=20, twistLimitDegrees=5}
 {bone1Name=dyn_grenade_02, bone2Name=dyn_grenade_01, swing1LimitDegrees=35, swing2LimitDegrees=10, twistLimitDegrees=5}
 {bone1Name=dyn_grenade_03, bone2Name=dyn_grenade_02, swing1LimitDegrees=25, swing2LimitDegrees=20, twistLimitDegrees=5}
]
